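# Sound cue for fence impacts during demolition, routed to the impacts sound class with small-object hit attenuation and limited concurrency.
# A quality-level node selects among four branches of differing complexity.
# The first mixes randomized medium impacts with randomized chain-link hits, and the second uses a single chain-link wave instead.
# The last two play only an optimized impact wave, and every branch uses pitch modulation for variation.
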SoundCue /Game/Sounds/Fort_Impact_Sounds/Demolition/Fence/Demo_Fence_Impact_Cue
root=QualityLevel_0
nodes [
 {id=Mixer_0, children [Modulator_1,Modulator_2], properties {inputVolume [1,1.2]}}
 {id=Mixer_1, children [Modulator_4,Modulator_5], properties {inputVolume [1,1.2]}}
 {id=Modulator_0, children [WavePlayer_4], properties {pitchMax=1.07, pitchMin=0.92}}
 {id=Modulator_1, children [Random_8], properties {pitchMax=1.07, pitchMin=0.92}}
 {id=Modulator_2, children [Random_0], properties {pitchMax=1.07, pitchMin=1}}
 {id=Modulator_3, children [WavePlayer_5], properties {pitchMax=1.07, pitchMin=0.92}}
 {id=Modulator_4, children [Random_2], properties {pitchMax=1.07, pitchMin=0.92}}
 {id=Modulator_5, children [WavePlayer_6], properties {pitchMax=1.07, pitchMin=1}}
 {id=QualityLevel_0, children [Mixer_0,Mixer_1,Modulator_0,Modulator_3]}
 {id=Random_0, children [WavePlayer_1,WavePlayer_2], properties {weights [1,1]}}
 {id=Random_2, children [WavePlayer_8,WavePlayer_9], properties {weights [1,1]}}
 {id=Random_8, children [WavePlayer_0,WavePlayer_3], properties {weights [1,1]}}
 {id=WavePlayer_0, properties {soundWaveAssetPtr=&/Fence_Impact_Med1}}
 {id=WavePlayer_1, properties {soundWaveAssetPtr=&/fence_chainlink_impact_ST_01}}
 {id=WavePlayer_2, properties {soundWaveAssetPtr=&/fence_chainlink_impact_ST_02}}
 {id=WavePlayer_3, properties {soundWaveAssetPtr=&/Fence_Impact_Med4}}
 {id=WavePlayer_4, properties {soundWaveAssetPtr=&/Fence_Impact_opt}}
 {id=WavePlayer_5, properties {soundWaveAssetPtr=&Fence_Impact_opt}}
 {id=WavePlayer_6, properties {soundWaveAssetPtr=&fence_chainlink_impact_ST_01}}
 {id=WavePlayer_8, properties {soundWaveAssetPtr=&Fence_Impact_Med1}}
 {id=WavePlayer_9, properties {soundWaveAssetPtr=&Fence_Impact_Med4}}
]
properties {
 attenuationSettings=/Game/Sounds/Attenuation/Fort_Demolition_Hit_Small_Sized_Object_Att
 maxDistance=4000
 soundClassObject=&Fort_Impacts
 volumeMultiplier=0.68
 concurrencyOverrides {maxCount=4}
 concurrencySet [&CONC_Object_Impacts_FourMax]
}
